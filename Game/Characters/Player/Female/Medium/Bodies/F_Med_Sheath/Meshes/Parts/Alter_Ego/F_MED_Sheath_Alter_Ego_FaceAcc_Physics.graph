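# Physics asset for the Alter Ego face accessory mesh of the medium female Sheath character, setting up dynamic hair.
# Capsules cover two right braids, a left front strand, a left side clump with many tips, plus head, neck, upper spine and right clavicle.
# Hair roots hang from the head on free-rotation constraints, with angle-limited joints down each chain.
# Extra free constraints cross-link neighboring side sections and one pair of tips.
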PhysicsAsset /Game/Characters/Player/Female/Medium/Bodies/F_Med_Sheath/Meshes/Parts/Alter_Ego/F_MED_Sheath_Alter_Ego_FaceAcc_Physics
bodies {
 clavicle_r [{shapeType=Capsule, length=5, radius=5.755, center=<-2.2967 0.5139 0.8891>, rotation=<rot 0 -89.9995 -82.4604>}]
 dyn_hair_braida_a_r [{shapeType=Capsule, length=1, radius=1}]
 dyn_hair_braida_b_r [{shapeType=Capsule, length=4.25, radius=1, center=<3.133 0 0>, rotation=<rot 90 0 0>}]
 dyn_hair_braida_c_r [{shapeType=Capsule, length=5, radius=1, center=<3.7392 0 0>, rotation=<rot 90 180 -175.7876>}]
 dyn_hair_braidb_a_r [{shapeType=Capsule, length=1, radius=1}]
 dyn_hair_braidb_b_r [{shapeType=Capsule, length=4, radius=1, center=<2.9154 0 0>, rotation=<rot 90 0 0>}]
 dyn_hair_braidb_c_r [{shapeType=Capsule, length=5.75, radius=1, center=<3.9934 0 0>, rotation=<rot 88.4709 -179.9999 180>}]
 dyn_hair_front_a_l [{shapeType=Capsule, length=1, radius=1}]
 dyn_hair_front_b_l [{shapeType=Capsule, length=2.25, radius=1, center=<1.5561 0 0>, rotation=<rot 90 0 0>}]
 dyn_hair_front_c_l [{shapeType=Capsule, length=3, radius=1, center=<2.3248 0 0>, rotation=<rot 90 0 0>}]
 dyn_hair_side_main_l [{shapeType=Capsule, length=1, radius=1}]
 dyn_hair_side_suba_l [{shapeType=Capsule, length=6, radius=1.75, center=<2.1675 0 0.221>, rotation=<rot 90 0 0>}]
 dyn_hair_side_subb_l [{shapeType=Capsule, length=4.5, radius=2.25, center=<2.359 0 0>, rotation=<rot 90 0 0>}]
 dyn_hair_side_subc_base_l [{shapeType=Capsule, length=1, radius=1}]
 dyn_hair_side_subc_l [{shapeType=Capsule, length=4.75, radius=2, center=<4.4485 0 0.0001>, rotation=<rot 90 0 0>}]
 dyn_hair_tip_a [{shapeType=Capsule, length=3.25, radius=1.25, center=<2.9811 0.8712 -0.9017>, rotation=<rot 76.227 28.7687 0.8379>}]
 dyn_hair_tip_b [{shapeType=Capsule, length=1.25, radius=0.75, center=<1.1921 0 0.971>, rotation=<rot 39.1345 180 180>}]
 dyn_hair_tip_c [{shapeType=Capsule, length=2, radius=0.75, center=<0.757 0.0416 0>, rotation=<rot 90 0 -9.9667>}]
 dyn_hair_tip_d [{shapeType=Capsule, length=3, radius=1.75, center=<2.4042 -0.3033 1.8511>, rotation=<rot 77.7445 -179.9996 170.8301>}]
 dyn_hair_tip_e [{shapeType=Capsule, length=3.25, radius=0.75, center=<2.1944 0.3578 0.823>, rotation=<rot 80.0138 -176.144 172.2198>}]
 dyn_hair_tip_f [{shapeType=Capsule, length=2.5, radius=1.25, center=<1.6484 -0.7221 1.0084>, rotation=<rot 49.8698 -166.1918 -135.6636>}]
 dyn_hair_tip_g [{shapeType=Capsule, length=3, radius=1.25, center=<2.6102 -0.4757 1.4924>, rotation=<rot 59.3334 179.9999 177.6609>}]
 dyn_hair_tip_h [{shapeType=Capsule, length=1.25, radius=0.75, center=<0.3809 0 0.7845>, rotation=<rot 30.0197 179.9999 179.9999>}]
 dyn_hair_tip_i [{shapeType=Capsule, length=1.75, radius=1, center=<1.0687 -0.3062 1.3488>, rotation=<rot 41.6479 143.3955 176.1852>}]
 dyn_hair_tip_j [{shapeType=Capsule, length=2.25, radius=1, center=<1.8914 -0.6352 0.7326>, rotation=<rot 45.7315 174.7041 -169.6655>}]
 dyn_hair_tip_k [{shapeType=Capsule, length=2.2189, radius=0.5933, center=<0.7743 -1.0799 -0.5361>, rotation=<rot 0 -149.7318 -70.8428>}]
 head [{shapeType=Capsule, length=4.5, radius=5.75, center=<2.5121 1.561 0.0758>, rotation=<rot 90 0 0>}]
 neck_02 [{shapeType=Capsule, length=4.25, radius=4.25, center=<-2.9725 -0.0649 0.0001>, rotation=<rot 90 0 -2.3828>}]
 spine_05 [{shapeType=Capsule, length=5.5, radius=11.75, center=<-0.1844 1.8005 0>}]
}
constraints [
 {
  bone1Name=dyn_hair_side_main_l
  bone2Name=head
  swing1LimitDegrees=45
  swing1Motion=Free
  swing2LimitDegrees=45
  swing2Motion=Free
  twistLimitDegrees=45
  twistMotion=Free
 }
 {
  bone1Name=dyn_hair_braida_a_r
  bone2Name=head
  swing1LimitDegrees=45
  swing1Motion=Free
  swing2LimitDegrees=45
  swing2Motion=Free
  twistLimitDegrees=45
  twistMotion=Free
 }
 {
  bone1Name=dyn_hair_braidb_a_r
  bone2Name=head
  swing1LimitDegrees=45
  swing1Motion=Free
  swing2LimitDegrees=45
  swing2Motion=Free
  twistLimitDegrees=45
  twistMotion=Free
 }
 {bone1Name=dyn_hair_side_subb_l, bone2Name=dyn_hair_side_main_l, swing1LimitDegrees=15, swing2LimitDegrees=15, twistLimitDegrees=5}
 {bone1Name=dyn_hair_side_suba_l, bone2Name=dyn_hair_side_main_l, swing1LimitDegrees=15, swing2LimitDegrees=15, twistLimitDegrees=5}
 {bone1Name=dyn_hair_side_subc_l, bone2Name=dyn_hair_side_subc_base_l, swing1LimitDegrees=15, swing2LimitDegrees=15, twistLimitDegrees=5}
 {bone1Name=dyn_hair_braida_b_r, bone2Name=dyn_hair_braida_a_r, swing1LimitDegrees=10, swing2LimitDegrees=15, twistLimitDegrees=5}
 {bone1Name=dyn_hair_braida_c_r, bone2Name=dyn_hair_braida_b_r, swing1LimitDegrees=10, swing2LimitDegrees=20, twistLimitDegrees=5}
 {bone1Name=dyn_hair_braidb_b_r, bone2Name=dyn_hair_braidb_a_r, swing1LimitDegrees=10, swing2LimitDegrees=10, twistLimitDegrees=5}
 {bone1Name=dyn_hair_braidb_c_r, bone2Name=dyn_hair_braidb_b_r, swing1LimitDegrees=10, swing2LimitDegrees=10, twistLimitDegrees=5}
 {
  bone1Name=dyn_hair_front_a_l
  bone2Name=head
  swing1LimitDegrees=45
  swing1Motion=Free
  swing2LimitDegrees=45
  swing2Motion=Free
  twistLimitDegrees=45
  twistMotion=Free
 }
 {bone1Name=dyn_hair_front_b_l, bone2Name=dyn_hair_front_a_l, swing1LimitDegrees=10, swing2LimitDegrees=10, twistLimitDegrees=5}
 {bone1Name=dyn_hair_front_c_l, bone2Name=dyn_hair_front_b_l, swing1LimitDegrees=10, swing2LimitDegrees=10, twistLimitDegrees=5}
 {bone1Name=dyn_hair_tip_a, bone2Name=dyn_hair_side_suba_l, swing1LimitDegrees=20, swing2LimitDegrees=15, twistLimitDegrees=5}
 {bone1Name=dyn_hair_tip_b, bone2Name=dyn_hair_side_suba_l, swing1LimitDegrees=40, swing2LimitDegrees=30, twistLimitDegrees=5}
 {bone1Name=dyn_hair_tip_c, bone2Name=dyn_hair_side_suba_l, swing1LimitDegrees=40, swing2LimitDegrees=30, twistLimitDegrees=5}
 {bone1Name=dyn_hair_tip_f, bone2Name=dyn_hair_side_subb_l, swing1LimitDegrees=40, swing2LimitDegrees=30, twistLimitDegrees=5}
 {bone1Name=dyn_hair_tip_e, bone2Name=dyn_hair_side_subb_l, swing1LimitDegrees=40, swing2LimitDegrees=30, twistLimitDegrees=5}
 {bone1Name=dyn_hair_tip_d, bone2Name=dyn_hair_side_subb_l, swing1LimitDegrees=40, swing2LimitDegrees=30, twistLimitDegrees=5}
 {bone1Name=dyn_hair_tip_i, bone2Name=dyn_hair_side_subc_l, swing1LimitDegrees=40, swing2LimitDegrees=30, twistLimitDegrees=5}
 {bone1Name=dyn_hair_tip_j, bone2Name=dyn_hair_side_subc_l, swing1LimitDegrees=15, swing2LimitDegrees=15, twistLimitDegrees=5}
 {bone1Name=dyn_hair_tip_g, bone2Name=dyn_hair_side_subc_l, swing1LimitDegrees=40, swing2LimitDegrees=30, twistLimitDegrees=5}
 {bone1Name=dyn_hair_tip_h, bone2Name=dyn_hair_side_subc_l, swing1LimitDegrees=40, swing2LimitDegrees=30, twistLimitDegrees=5}
 {
  bone1Name=head
  bone2Name=neck_02
  swing1LimitDegrees=45
  swing1Motion=Free
  swing2LimitDegrees=45
  swing2Motion=Free
  twistLimitDegrees=45
  twistMotion=Free
 }
 {
  bone1Name=neck_02
  bone2Name=spine_05
  swing1LimitDegrees=45
  swing1Motion=Free
  swing2LimitDegrees=45
  swing2Motion=Free
  twistLimitDegrees=45
  twistMotion=Free
 }
 {
  bone1Name=clavicle_r
  bone2Name=spine_05
  swing1LimitDegrees=45
  swing1Motion=Free
  swing2LimitDegrees=45
  swing2Motion=Free
  twistLimitDegrees=45
  twistMotion=Free
 }
 {
  bone1Name=dyn_hair_side_subb_l
  bone2Name=dyn_hair_side_subc_l
  swing1LimitDegrees=45
  swing1Motion=Free
  swing2LimitDegrees=45
  swing2Motion=Free
  twistLimitDegrees=45
  twistMotion=Free
 }
 {
  bone1Name=dyn_hair_side_suba_l
  bone2Name=dyn_hair_side_subb_l
  swing1LimitDegrees=45
  swing1Motion=Free
  swing2LimitDegrees=45
  swing2Motion=Free
  twistLimitDegrees=45
  twistMotion=Free
 }
 {
  bone1Name=dyn_hair_tip_f
  bone2Name=dyn_hair_tip_g
  swing1LimitDegrees=45
  swing1Motion=Free
  swing2LimitDegrees=45
  swing2Motion=Free
  twistLimitDegrees=45
  twistMotion=Free
 }
 {
  bone1Name=dyn_hair_side_subc_base_l
  bone2Name=head
  swing1LimitDegrees=45
  swing1Motion=Free
  swing2LimitDegrees=45
  swing2Motion=Free
  twistLimitDegrees=45
  twistMotion=Free
 }
 {bone1Name=dyn_hair_tip_k, bone2Name=dyn_hair_side_subc_l, swing1LimitDegrees=40, swing2LimitDegrees=20, twistLimitDegrees=5}
]
