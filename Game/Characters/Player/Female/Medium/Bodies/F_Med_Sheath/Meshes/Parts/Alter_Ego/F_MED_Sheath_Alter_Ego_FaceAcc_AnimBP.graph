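# Post-copy Anim Blueprint for the Alter Ego face accessory on a female medium player character, adding secondary motion to hair.
# It copies the attached parent's pose, then additively rotates a left side hair base bone as head rotation relative to spine_05, read each update, goes negative.
# Chained AnimDynamics spring nodes move front hair, braids and side hair, followed by a RigidBody pass using a dedicated physics asset in spine_05 space.
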AnimBlueprint /Game/Characters/Player/Female/Medium/Bodies/F_Med_Sheath/Meshes/Parts/Alter_Ego/F_MED_Sheath_Alter_Ego_FaceAcc_AnimBP
target_skeleton=&/F_MED_Sheath_Alter_Ego_FaceAcc_Skeleton
graphs [
 AnimGraph {nodes [
   Anim Dynamics - Bone: dyn_hair_side_subc_l(Alpha=0, ExternalForce="0,0,0") {
    angularDampingOverride=1
    angularSpringConstant=1500
    bAngularSpring=true
    bGravityOverrideInSimSpace=true
    bOverrideAngularDamping=true
    gravityScale=0.35
    simulationSpace=RootRelative
    alphaBoolBlend {blendOption=Linear}
    boundBone {boneName=dyn_hair_side_subc_l}
    componentLinearAccScale=<0.3 0.3 0.3>
    componentLinearVelScale=<0.3 0.3 0.3>
    physicsBodyDefinitions [{
      sphereCollisionRadius=0
      boundBone {boneName=dyn_hair_side_subc_l}
      boxExtents=<10 20 20>
      constraintSetup {angularLimitsMax=<0 0 0>, angularLimitsMin=<0 -20 -5>, angularTarget=<0 0 0>}
      localJointOffset=<10 0 0>
     }]
   }
   n1 = "Transform (Modify) Bone - Bone: dyn_hair_side_subc_base_l"(
    ComponentPose=Local To Component(LocalPose=Copy Pose From Mesh() {bUseAttachedParent=true})
    Alpha=Get headFwdBk() -> (headFwdBk:double)
    Translation="0,0,0"
    Rotation="20,0,0"
    Scale="1,1,1"
   ) {
    rotationMode=BMM_Additive
    rotationSpace=BCS_BoneSpace
    translationMode=BMM_Additive
    translationSpace=BCS_BoneSpace
    alphaBoolBlend {blendOption=Linear}
    alphaScaleBiasClamp {bMapRange=true, inRange {max=0, min=-40}, outRange {max=0, min=1}}
    boneToModify {boneName=dyn_hair_side_subc_base_l}
    rotation {pitch=20}
   }
   n2 = Anim Dynamics - Bone: dyn_hair_front_a_l(ComponentPose=@n1, Alpha=1, ExternalForce="0,0,0") {
    angularDampingOverride=1
    angularSpringConstant=1500
    bAngularSpring=true
    bGravityOverrideInSimSpace=true
    bOverrideAngularDamping=true
    gravityScale=0.35
    simulationSpace=RootRelative
    alphaBoolBlend {blendOption=Linear}
    boundBone {boneName=dyn_hair_front_a_l}
    componentLinearAccScale=<0.4 0.4 0.4>
    componentLinearVelScale=<0.4 0.4 0.4>
    physicsBodyDefinitions [{
      sphereCollisionRadius=0
      boundBone {boneName=dyn_hair_front_a_l}
      boxExtents=<10 20 20>
      constraintSetup {angularLimitsMax=<0 0 30>, angularLimitsMin=<0 -5 0>, angularTarget=<0 0 0>}
      localJointOffset=<10 0 0>
     }]
   }
   n3 = Anim Dynamics - Bone: dyn_hair_braida_a_r(ComponentPose=@n2, Alpha=1, ExternalForce="0,0,0") {
    angularDampingOverride=1
    angularSpringConstant=1500
    bAngularSpring=true
    bGravityOverrideInSimSpace=true
    bOverrideAngularDamping=true
    gravityScale=0.35
    simulationSpace=RootRelative
    alphaBoolBlend {blendOption=Linear}
    boundBone {boneName=dyn_hair_braida_a_r}
    componentLinearAccScale=<0.4 0.4 0.5>
    componentLinearVelScale=<0.4 0.4 0.5>
    physicsBodyDefinitions [{
      sphereCollisionRadius=0
      boundBone {boneName=dyn_hair_braida_a_r}
      boxExtents=<10 20 20>
      constraintSetup {angularLimitsMax=<0 10 10>, angularLimitsMin=<0 -5 -5>, angularTarget=<0 0 0>, linearAxesMax=<0 2 1>}
      localJointOffset=<10 0 0>
     }]
   }
   n4 = Anim Dynamics - Bone: dyn_hair_braidb_a_r(ComponentPose=@n3, Alpha=1, ExternalForce="0,0,0") {
    angularDampingOverride=1
    angularSpringConstant=1500
    bAngularSpring=true
    bGravityOverrideInSimSpace=true
    bOverrideAngularDamping=true
    gravityScale=0.35
    simulationSpace=RootRelative
    alphaBoolBlend {blendOption=Linear}
    boundBone {boneName=dyn_hair_braidb_a_r}
    componentLinearAccScale=<0.4 0.4 0.4>
    componentLinearVelScale=<0.4 0.4 0.4>
    physicsBodyDefinitions [{
      sphereCollisionRadius=0
      boundBone {boneName=dyn_hair_braidb_a_r}
      boxExtents=<10 20 20>
      constraintSetup {angularLimitsMax=<0 2 10>, angularLimitsMin=<0 -3 -10>, angularTarget=<0 0 0>}
      localJointOffset=<10 0 0>
     }]
   }
   n5 = Anim Dynamics - Bone: dyn_hair_side_main_l(ComponentPose=@n4, Alpha=0.5, ExternalForce="0,0,0") {
    angularDampingOverride=1
    angularSpringConstant=1500
    bAngularSpring=true
    bGravityOverrideInSimSpace=true
    bOverrideAngularDamping=true
    gravityScale=0.35
    simulationSpace=RootRelative
    alphaBoolBlend {blendOption=Linear}
    boundBone {boneName=dyn_hair_side_main_l}
    componentLinearAccScale=<0.3 0.3 0.3>
    componentLinearVelScale=<0.3 0.3 0.3>
    physicsBodyDefinitions [{
      sphereCollisionRadius=0
      boundBone {boneName=dyn_hair_side_main_l}
      boxExtents=<10 20 20>
      constraintSetup {angularLimitsMax=<0 0 5>, angularLimitsMin=<0 -10 -5>, angularTarget=<0 0 0>}
      localJointOffset=<10 0 0>
     }]
   }
   n6 = RigidBody(ComponentPose=@n5) {
    bClampLinearTranslationLimitToRefPose=true
    bOverrideWorldGravity=true
    overridePhysicsAsset=&/F_MED_Sheath_Alter_Ego_FaceAcc_Physics
    simulationSpace=BaseBoneSpace
    alphaBoolBlend {blendOption=Linear}
    baseBoneRef {boneName=spine_05}
    componentLinearAccScale=<0.35 0.4 0.4>
    componentLinearVelScale=<0.35 0.4 0.4>
    overrideWorldGravity {z=-980}
   }
   Output Pose(Result=Component To Local(ComponentPose=@n6)) {layerGroup=None}
  ]}
 Begin Target {nodes [Begin Target() {is=title K2Node_FunctionEntry}]}
 End Target {nodes [End Target() {is=title K2Node_FunctionEntry}]}
 EventGraph {nodes [
   Event BlueprintUpdateAnimation()
   n1 = GetDeltaTransformFromRefPose(self=GetOwnerSkeletalMeshComponent(), BoneName=head, BaseName=spine_05)
   > Set headFwdBk(headFwdBk:double=BreakRotator(InRot=BreakTransform(InTransform=@n1).Rotation).Yaw) -> (Output_Get:double)
   Event OnBeginTargeting() {is=title K2Node_Event}
   > Call Begin Target!(execute:Exec, self:F MED Sheath Alter Ego Face Acc Anim BP Object Reference) -> (then:Exec) {is=title K2Node_CallDelegate}
   Event OnEndTargeting() {is=title K2Node_Event}
   > Call End Target!(execute:Exec, self:F MED Sheath Alter Ego Face Acc Anim BP Object Reference) -> (then:Exec) {is=title K2Node_CallDelegate}
   Event BlueprintBeginPlay() {is=title K2Node_Event}
   > Cast To PlayerPawn_Athena(Object=TryGetPawnOwner())
  ]}
]
variables [
 Begin Target {default=()}
 End Target {default=()}
 headFwdBk {type=number}
]
